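; Caretaker monster NPC controller Blueprint that adds an optional periodic debug update.
; On BeginPlay, if the Debug_Update flag is set, it starts a looping timer at the UpdateTime interval.
; Each timer tick runs the DebugTimer event, which calls the CaretakerController DebugUpdate function with UpdateTime as its argument.
; The construction script is empty.
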(event EventBeginPlay
  (if (Variables|Default|GetDebug_Update)
    (Utilities|Time|SetTimerbyEvent (delegate Custom|DebugTimer) (Variables|Default|GetUpdateTime) true false 1.0)))

(event Custom|DebugTimer
  (CaretakerController|DebugUpdate (Variables|Default|GetUpdateTime)))

(fn ConstructionScript)
